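# First-person fire sound cue for the Boltaction_HT sniper rifle.
# It picks one of three equally weighted close-range fire waves at random, then applies a slight upward pitch variation at constant volume.
# A first-person projectile attenuation, a first-person ranged-fire sound class and a first-person gunfire concurrency set govern playback.
# Its concurrency overrides specify an owner-limited cap that stops the oldest sound, and its debug flag is on.
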SoundCue /Game/Sounds/Fort_Weapons/SniperRifles/Boltaction_HT/BoltAction_HT_Fire_1P_Cue
root=Modulator_2
nodes [
 {id=Modulator_2, children [Random_1], properties {pitchMax=1.1, pitchMin=1, volumeMax=1, volumeMin=1}}
 {id=Random_1, children [WavePlayer_0,WavePlayer_1,WavePlayer_2], properties {weights [1,1,1]}}
 {id=WavePlayer_0, properties {soundWaveAssetPtr=&/BoltAction_HT_Fire_Close_01}}
 {id=WavePlayer_1, properties {soundWaveAssetPtr=&/BoltAction_HT_Fire_Close_02}}
 {id=WavePlayer_2, properties {soundWaveAssetPtr=&/BoltAction_HT_Fire_Close_03}}
]
properties {
 attenuationSettings=&Fort_Weapon_Projectile_FirstPerson_Athena_Att
 bDebug=true
 maxDistance=2097152
 soundClassObject=&Fort_WeaponRanged_Fire_1P_PSM
 volumeMultiplier=1
 concurrencyOverrides {bLimitToOwner=true, maxCount=3, resolutionRule=StopOldest}
 concurrencySet [&Weapon_Gunfire_FirstPerson_Conc]
}
