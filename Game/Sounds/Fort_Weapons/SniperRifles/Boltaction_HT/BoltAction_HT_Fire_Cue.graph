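# Sound cue for the firing shot of the Boltaction_HT sniper rifle, layered by distance.
# A mixer blends close, medium and distant layers, each picking randomly from three waves, then passing through a modulator, its own attenuation and a sound class override.
# Close uses a third-person fire class with slight pitch variation, while medium and distant share a distant class with boosted volume.
# The cue references local and global gunfire concurrency sets.
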SoundCue /Game/Sounds/Fort_Weapons/SniperRifles/Boltaction_HT/BoltAction_HT_Fire_Cue
root=Mixer_0
nodes [
 {id=Attenuation_2, children [Modulator_2], properties {attenuationSettings=&Fort_Weapon_Projectile_Close_Athena_Att}}
 {
  id=Attenuation_3
  children [Modulator_3]
  properties {attenuationSettings=&Fort_Weapon_Projectile_Medium_Mono_Athena_Att}
 }
 {
  id=Attenuation_4
  children [Modulator_4]
  properties {attenuationSettings=&Fort_Weapon_Projectile_Distant_Mono_Athena_Att}
 }
 {id=Mixer_0, children [SoundClass_2,SoundClass_4,SoundClass_3], properties {inputVolume [1,0.9,0.7]}}
 {id=Modulator_2, children [Random_2], properties {pitchMax=1.1, pitchMin=1, volumeMax=1, volumeMin=1}}
 {id=Modulator_3, children [Random_3], properties {volumeMax=1.7, volumeMin=1.6}}
 {id=Modulator_4, children [Random_1], properties {volumeMax=2.3, volumeMin=2.1}}
 {id=Random_1, children [WavePlayer_10, WavePlayer_11, WavePlayer_12], properties {weights [1,1,1]}}
 {id=Random_2, children [WavePlayer_0,WavePlayer_1,WavePlayer_2], properties {weights [1,1,1]}}
 {id=Random_3, children [WavePlayer_6,WavePlayer_7,WavePlayer_8], properties {weights [1,1,1]}}
 {
  id=SoundClass_2
  children [Attenuation_2]
  properties {soundClassOverride=&Fort_WeaponRanged_Fire_3P_PSM}
 }
 {
  id=SoundClass_3
  children [Attenuation_4]
  properties {soundClassOverride=&Fort_WeaponRanged_Fire_Distant}
 }
 {
  id=SoundClass_4
  children [Attenuation_3]
  properties {soundClassOverride=&Fort_WeaponRanged_Fire_Distant}
 }
 {id=WavePlayer_0, properties {soundWaveAssetPtr=&/BoltAction_HT_Fire_Close_01}}
 {id=WavePlayer_1, properties {soundWaveAssetPtr=&/BoltAction_HT_Fire_Close_02}}
 {id=WavePlayer_10, properties {soundWaveAssetPtr=&/BoltAction_HT_Fire_Dist_01}}
 {id=WavePlayer_11, properties {soundWaveAssetPtr=&/BoltAction_HT_Fire_Dist_02}}
 {id=WavePlayer_12, properties {soundWaveAssetPtr=&/BoltAction_HT_Fire_Dist_03}}
 {id=WavePlayer_2, properties {soundWaveAssetPtr=&/BoltAction_HT_Fire_Close_03}}
 {id=WavePlayer_6, properties {soundWaveAssetPtr=&/BoltAction_HT_Fire_Med_01}}
 {id=WavePlayer_7, properties {soundWaveAssetPtr=&/BoltAction_HT_Fire_Med_02}}
 {id=WavePlayer_8, properties {soundWaveAssetPtr=&/BoltAction_HT_Fire_Med_03}}
]
properties {
 bDebug=true
 maxDistance=27500
 volumeMultiplier=0.7
 concurrencyOverrides {bLimitToOwner=true, maxCount=3, resolutionRule=StopOldest}
 concurrencySet [&Weapon_Gunfire_Local_Conc, &Weapon_Gunfire_Global_Conc]
}
