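# Sound cue for the end of the reload on the Boltaction_HT bolt-action sniper rifle.
# A single reload-end wave plays through a weapon foley sound class override, a modulator that randomizes volume at a reduced level, and weapon reload attenuation.
# The cue also sets that class and attenuation at top level, lowers pitch and volume slightly, overrides concurrency, and has its debug flag on.
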SoundCue /Game/Sounds/Fort_Weapons/SniperRifles/Boltaction_HT/BoltAction_HT_Reload_End_Cue
root=Attenuation_0
nodes [
 {id=Attenuation_0, children [Modulator_0], properties {attenuationSettings=&Fort_Weapon_Reload_Att}}
 {id=Modulator_0, children [SoundClass_0], properties {volumeMax=0.6, volumeMin=0.5}}
 {id=SoundClass_0, children [WavePlayer_0], properties {soundClassOverride=&Fort_WeaponFoley}}
 {id=WavePlayer_0, properties {soundWaveAssetPtr=&/Boltaction_HT_Reload_End}}
]
properties {
 attenuationSettings=&Fort_Weapon_Reload_Att
 bDebug=true
 bOverrideConcurrency=true
 maxDistance=2250
 pitchMultiplier=0.9
 soundClassObject=&Fort_WeaponFoley
 volumeMultiplier=0.902
}
