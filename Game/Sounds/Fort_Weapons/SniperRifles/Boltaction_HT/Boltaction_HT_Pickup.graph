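# Sound cue for the pickup sound of the Boltaction_HT sniper rifle.
# A single wave player feeds a Modulator that applies a fixed volume boost, with a reduced cue volume multiplier on top.
# It uses the weapon pickup attenuation and loot-collect sound class.
# A concurrency override allows only one instance at a time and blocks new plays while one is active.
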SoundCue /Game/Sounds/Fort_Weapons/SniperRifles/Boltaction_HT/Boltaction_HT_Pickup
root=Modulator_1
nodes [
 {id=Modulator_1, children [WavePlayer_0], properties {volumeMax=1.2, volumeMin=1.2}}
 {id=WavePlayer_0, properties {soundWaveAssetPtr=&/BoltAction_HT_Pickup_01}}
]
properties {
 attenuationSettings=&Fort_Weapon_Pickup_Att
 bOverrideConcurrency=true
 maxDistance=3500
 soundClassObject=&Fort_Loot_Collect
 volumeMultiplier=0.4615
 concurrencyOverrides {maxCount=1, resolutionRule=PreventNew}
}
